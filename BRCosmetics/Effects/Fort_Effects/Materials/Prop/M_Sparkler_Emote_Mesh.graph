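# Lit mesh-particle material for the sparkler emote prop, pairing a plain grey surface look with a colored glow toward one end.
# An inverted V gradient, offset down and clamped, confines the glow to part of the UV range.
# That mask is multiplied by particle color and passed through the shared FX emissive adjustment function.
# A tiled noise normal, softened by a flatness parameter, supplies surface detail.
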Material /BRCosmetics/Effects/Fort_Effects/Materials/Prop/M_Sparkler_Emote_Mesh
graph {
 properties {bUsedWithMeshParticles=true, pixelDepthOffsetMode=PDOM_Legacy, usedShadingModels=Default Lit}
 nodes [
  Constant3Vector() {constant=<rgba 0.1927 0.1927 0.1927 0>}
  Constant()
  Constant() {r=0.5}
  MaterialFunctionCall(
   Normal=TextureSample(UVs=TextureCoordinate() {uTiling=4, vTiling=4}) {samplerType=SAMPLERTYPE_Normal, texture=/Game/Global/Textures/Default/Noise/T_Noise_Normal_01}.RGB
   Flatness=ScalarParameter() {defaultValue=0.75, parameterName="Flatness (S)"}
  ) {materialFunction=&FlattenNormal}
  n1 = OneMinus(None=MaterialFunctionCall() {materialFunction=&LinearGradient}.VGradient)
  MaterialFunctionCall(In=@n1, Contrast=Constant() {r=0.5}) {materialFunction=&CheapContrast}
  n2 = Add(A=@n1, B=Constant() {r=-0.6})
  MaterialFunctionCall(Emissive=Multiply(A=Clamp(None=@n2), B=ParticleColor().RGB)) {materialFunction=&MF_FX_EmissiveAdjustments}
 ]
}
